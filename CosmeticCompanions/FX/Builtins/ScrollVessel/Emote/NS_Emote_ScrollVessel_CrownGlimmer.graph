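# Niagara effect for the ScrollVessel cosmetic companion's CrownGlimmer emote, a brief flash of light.
# Flare spawns one warm, camera-offset sprite that fades in and out while shrinking to nothing.
# Rays spawns short-lived, randomly oriented white meshes that lengthen and thin as they fade.
# Both scale alpha by time of day through FNTimeOfDayScalar; LWE_Glow and SparklesLWE show no modules or renderers.
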NiagaraSystem /CosmeticCompanions/FX/Builtins/ScrollVessel/Emote/NS_Emote_ScrollVessel_CrownGlimmer
dependencies {usedDataInterfaces [&NiagaraDataInterfaceCurve], usedDynamicInputs [], usedModules [&SystemState]}
emitters {
 Flare {
  emitterUpdateScript [
   &EmitterState(Life Cycle Mode=::NewEnumerator1, Loop Behavior=::NewEnumerator1, Loop Duration=1, UseLoopDelay=false)
   &SpawnBurst_Instantaneous(Spawn Count=1, Use Spawn Probability=false)
  ]
  particleSpawnScript [
   &InitializeParticle(
    Color=&MakeLinearColorFromVectorAndFloat(
     "Vector (RGB)"=&MakeVectorFromLinearColorRGB(Color=<rgba 2 1.1222 0.4294 1>)
     "Float (Alpha)"=&FNTimeOfDayScalar(Night Scale=0.5, Morning Scale=0.3, Evening Scale=0.4)
    )
    Lifetime=0.52
    Sprite Rotation Mode=::NewEnumerator2
    Sprite Size Mode=::NewEnumerator3
    Uniform Sprite Size=60
    UsePositionOffset=false
    Write Material Random=true
   )
   &DynamicMaterialParameters(
    Index 0 Param 1=&RandomRangeFloat(Minimum=0.4, Maximum=0.6)
    Index 0 Param 2=1
    Index 0 Param 3=1
    Index 0 Param 4=1
    Module.Param0WriteEnabled=true
    Module.Param1WriteEnabled=true
    Module.Param2WriteEnabled=true
    Module.Param3WriteEnabled=true
   )
  ]
  particleUpdateScript [
   &ParticleState()
   &ScaleColor(
    Scale Alpha=&FloatFromCurve(FloatCurve=DataInterface() {propertyValues {Curve {
        defaultValue=0
        keys [
         {arriveTangent=4.7871, interpMode=RCIM_Cubic, leaveTangent=4.7871, tangentMode=RCTM_User, time=0, value=0}
         {interpMode=RCIM_Cubic, tangentMode=RCTM_User, time=0.4099, value=1}
         {arriveTangent=-3.0525, tangentMode=RCTM_User, time=1, value=0}
        ]
       }}})
    ScaleA=true
    ScaleRGB=true
   )
   &ScaleSpriteSize(Scale Factor=&Vector2DFromFloat(Value=&FloatFromCurve(FloatCurve=DataInterface() {propertyValues {Curve {keys [<at 0 1.5 RCIM_Cubic>,<at 1 0>]}}})))
   &CameraOffset(Camera Offset Amount=9)
  ]
  rendererClasses [&NiagaraSpriteRendererProperties]
  renderers [{rendererClass=&NiagaraSpriteRendererProperties}]
 }
 LWE_Glow {rendererClasses [], renderers []}
 Rays {
  emitterUpdateScript [
   &EmitterState(Life Cycle Mode=::NewEnumerator1, Loop Behavior=::NewEnumerator1, Loop Duration=0.1, UseLoopDelay=false)
   &SpawnRate(SpawnRate=90, Use Spawn Probability=false)
   &SpawnBurst_Instantaneous(Spawn Count=3, Use Spawn Probability=false)
  ]
  particleSpawnScript [
   &InitializeParticle(
    Color=&MakeLinearColorFromVectorAndFloat(
     "Vector (RGB)"=<1 1 1>
     "Float (Alpha)"=&Multiply_Float(
      A=&FNTimeOfDayScalar(Night Scale=0.5, Morning Scale=0.3, Evening Scale=0.4)
      B=&RandomRangeFloat(Minimum=0.565, Maximum=0.748)
     )
    )
    Lifetime Max=0.52
    Lifetime Min=0.13
    Lifetime Mode=::NewEnumerator1
    Mesh Scale Max=<0.52 0.15 0.15>
    Mesh Scale Min=<0.37 0.1 0.1>
    Mesh Scale Mode=::NewEnumerator2
    UsePositionOffset=false
    Write Material Random=true
   ) {module=InitializeParticle001}
   &CameraOffset(Camera Offset Amount=-4)
   &DynamicMaterialParameters(
    Index 0 Param 1=&UniformRangedFloat(Minimum=-1, Maximum=1)
    Index 0 Param 2=1
    Module.Param0WriteEnabled=true
    Module.Param1WriteEnabled=true
    Module.Param2WriteEnabled=true
    Module.Param3WriteEnabled=true
   )
   &InitialMeshOrientation(
    Rotation=&Add_Vector(A=<0 0 0>, B=&RandomRangeVector(Minimum=<-1 0 -1>, Maximum=<1 1 1>))
    Use Rotation Vector=true
   ) {module=InitialMeshOrientation001}
   &SystemLocation(MaskSpawnGroup=false)
  ]
  particleUpdateScript [
   &ParticleState()
   &SolveForcesAndVelocity(Clamp Velocity=false, Limit Acceleration=false)
   &ScaleColor(
    Scale Alpha=&Multiply_Float(
     A=&FloatFromCurve(FloatCurve=DataInterface() {propertyValues {Curve {keys [
          {arriveTangent=4.2233, interpMode=RCIM_Cubic, leaveTangent=4.2233, tangentMode=RCTM_User, time=0, value=0}
          {arriveTangent=-0.4315, interpMode=RCIM_Cubic, leaveTangent=-0.4315, tangentMode=RCTM_User, time=0.2593, value=0.4527}
          {arriveTangent=-1.144, interpMode=RCIM_Cubic, leaveTangent=-1.144, tangentMode=RCTM_User, time=0.623, value=0.1632}
          {tangentMode=RCTM_User, time=1, value=0}
         ]}}})
     B=&RandomRangeFloat(Minimum=1, Maximum=2.2)
    )
    ScaleA=true
    ScaleRGB=true
   )
   &ScaleMeshSize(Scale Factor=&MakeVector(
     X=&FloatFromCurve(FloatCurve=DataInterface() {propertyValues {Curve {keys [<at 0 0.8>,<at 1 1.2>]}}})
     Y=1
     Z=&FloatFromCurve(FloatCurve=DataInterface() {propertyValues {Curve {keys [{arriveTangent=-2.6122, interpMode=RCIM_Cubic, leaveTangent=-2.6122, tangentMode=RCTM_User, time=0, value=1.8}, <at 1 0.6>]}}})
    ))
  ]
  rendererClasses [&NiagaraMeshRendererProperties]
  renderers [{rendererClass=&NiagaraMeshRendererProperties, rendererIndex=0}]
 }
 SparklesLWE {rendererClasses [], renderers []}
}
summary {
 systemName=NS_Emote_ScrollVessel_CrownGlimmer
 emitters [
  {emitterName=LWE_Glow, rendererClasses []}
  {emitterName=Flare, rendererClasses [&NiagaraSpriteRendererProperties]}
  {emitterName=Rays, rendererClasses [&NiagaraMeshRendererProperties]}
  {emitterName=SparklesLWE, rendererClasses []}
 ]
 userVariables []
}
