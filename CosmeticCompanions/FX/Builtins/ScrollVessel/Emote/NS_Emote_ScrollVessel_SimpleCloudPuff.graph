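# Niagara system for a simple cloud puff effect, named as part of the ScrollVessel cosmetic companion's emote FX.
# It holds a single emitter, SmokeLWE, whose name suggests a lightweight smoke emitter; no renderers are listed for it.
# SystemState is the only listed module.
# A User.NormalizedNotifyProgress float, defaulting to zero, is exposed, apparently for animation notify progress, though no use of it is shown.
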NiagaraSystem /CosmeticCompanions/FX/Builtins/ScrollVessel/Emote/NS_Emote_ScrollVessel_SimpleCloudPuff
dependencies {usedDataInterfaces [], usedDynamicInputs [], usedModules [&SystemState]}
emitters {SmokeLWE {rendererClasses [], renderers []}}
summary {
 systemName=NS_Emote_ScrollVessel_SimpleCloudPuff
 emitters [{emitterName=SmokeLWE, rendererClasses []}]
 userVariables [User.NormalizedNotifyProgress {type=NiagaraFloat, defaultValue=0, description=""}]
}
